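# Niagara system for the Mystic skydiving dive contrail, an orange, fiery effect attached to the character's skeletal mesh.
# Spinning meshes with a light form the center and mirrored spinning meshes and glows follow both hands, joined by ribbons, GPU sparks and wave sprites.
# System dive values, notably an FN_Integrator output, blend offsets, spawn rates, velocities, size and opacity.
# Emitter-age curves stage the build-up, and front-end modifiers substitute separate values outside gameplay.
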NiagaraSystem /BRCosmetics/Athena/Effects/Particles/Skydiving/NS_Contrail_Mystic/FX/NS_Contrail_Dive_Mystic
dependencies {
 usedDataInterfaces [
  &NiagaraDataInterfaceMeshRendererInfo
  &NiagaraDataInterfaceVectorCurve
  &NiagaraDataInterfaceCurve
  &NiagaraDataInterfaceArrayColor
  &NiagaraDataInterfaceColorCurve
 ]
 usedDynamicInputs [
  &Divide_Float
  &MakeFloatFromInt
  &ReturnParticleIDAttributes
  &SetBoolByFloatComparison
  &Subtract_Float
 ]
 usedModules [
  &SystemState
  /Game/Effects/Niagara/Modules/Skydiving/FN_DivingAttributes/v5/FN_DivingAttributes
  &self:SystemSpawnScript.SystemScriptSource.SystemScriptGraph.NiagaraNodeAssignment_0.SetVariables_#1
  &self:SystemSpawnScript.SystemScriptSource.SystemScriptGraph.NiagaraNodeAssignment_1.SetVariables_#2
  &FN_Integrator
  &self:SystemSpawnScript.SystemScriptSource.SystemScriptGraph.NiagaraNodeAssignment_2.SetVariables_#3
  &self:SystemSpawnScript.SystemScriptSource.SystemScriptGraph.NiagaraNodeAssignment_5.SetVariables_#4
 ]
}
emitters {
 Center {
  emitterSpawnScript [&FN_PreserveSkydiveUserVariables() {module=PreserveSkydiveUserVariables}]
  emitterUpdateScript [
   &EmitterState(
    Life Cycle Mode=::NewEnumerator1
    Loop Behavior=::NewEnumerator1
    Loop Duration Mode=::NewEnumerator1
    Reset Age On Awaken=true
    Scalability Mode=::NewEnumerator1
    UseLoopDelay=false
   )
   &SpawnBurst_Instantaneous(Spawn Count=1, Use Spawn Probability=false)
  ]
  particleSpawnScript [
   &FN_SetSpawnGroup(Number of Spawn Groups=1) {module=FN_SetSpawnGroup001}
   &InitializeParticle(
    Color=<rgba 12 1.6826 0.2292 1>
    Mesh Scale=<1.4 1.4 -4>
    Mesh Scale Mode=::NewEnumerator4
    Sprite Size Mode=::NewEnumerator3
    Uniform Sprite Size=200
    UsePositionOffset=false
    Write Material Random=true
   )
   &InitialMeshOrientation(
    Orientation Axis=<0 0 1>
    Orientation Coordinate Space=::Local
    Orientation Vector=<0 0 1>
    Use Orientation Vector=true
    Use Rotation Vector=false
   )
  ]
  particleUpdateScript [
   /Game/Effects/Niagara/Modules/Skydiving/FN_DivingAttributes/FN_DivingAttributes_v2(Accumulate Dive=0.1, Accumulate Rest=0.1, SkeletalMesh:NiagaraDataInterfaceSkeletalMesh=$User.SkeletalMesh) {module=FN_DivingAttributes_v2001}
   &ParticleState(Kill Particles When Lifetime Has Elapsed=false, Let Infinitely Lived Particles Die When Emitter Deactivates=true)
   &SkeletalMeshLocation(
    Bone / Socket Index=5
    Bone Sampling Mode=::NewEnumerator4
    Inherit Velocity=false
    Interpolate Mesh Sampling=false
    MaskSpawnGroup=false
    Sampled Position Offset=&Lerp_Vector(A=<0 0 32>, B=<1 1 25>, Alpha=$System.FN_Integrator001.Value)
    Sampled Position Offset Coordinate Space:ENiagaraMeshTransforms=::NewEnumerator2
    Skeletal Mesh=$User.SkeletalMesh
   )
   &SolveForcesAndVelocity(Clamp Velocity=false, Limit Acceleration=false)
   &ScaleMeshSize(Scale Factor=&VectorFromCurve(
     VectorCurve=DataInterface() {propertyValues {
       XCurve {keys [
         {arriveTangent=4.4809, interpMode=RCIM_Cubic, leaveTangent=4.4809, tangentMode=RCTM_User, time=0.1, value=0}
         <at 0.4 0.475>
         {arriveTangent=0.875, interpMode=RCIM_Cubic, leaveTangent=0.875, time=0.6, value=0.475}
         {tangentMode=RCTM_User, time=1, value=1}
        ]}
       YCurve {keys [
         {arriveTangent=4.4809, interpMode=RCIM_Cubic, leaveTangent=4.4809, tangentMode=RCTM_User, time=0.1, value=0}
         <at 0.4 0.475>
         {arriveTangent=0.875, interpMode=RCIM_Cubic, leaveTangent=0.875, time=0.6, value=0.475}
         {tangentMode=RCTM_User, time=1, value=1}
        ]}
       ZCurve {keys [{tangentMode=RCTM_User, time=0, value=1.1}, {tangentMode=RCTM_User, time=1.1, value=1}]}
      }}
     CurveIndex=$Emitter.Age
    ))
   &UpdateMeshOrientation(
    Rotation Coordinate Space=::NewEnumerator2
    Rotation Rate=&Lerp_Float(A=0.09, B=0.3, Alpha=$System.FN_Integrator001.Value)
    Rotation Vector=<0 0 1>
   )
   &ScaleColor(
    Scale Alpha=&FloatFromCurve(
     FloatCurve=DataInterface() {propertyValues {Curve {keys [
         {arriveTangent=4.0516, interpMode=RCIM_Cubic, leaveTangent=4.0516, tangentMode=RCTM_User, time=0, value=0}
         {arriveTangent=0.0961, interpMode=RCIM_Cubic, leaveTangent=0.0961, tangentMode=RCTM_User, time=0.25, value=1}
         {arriveTangent=0.2272, interpMode=RCIM_Cubic, leaveTangent=0.2272, tangentMode=RCTM_User, time=1, value=1}
        ]}}}
     Scale Curve=&Lerp_Float(A=0.2, B=0.5, Alpha=$System.Dive)
    )
    ScaleA=true
    ScaleRGB=false
   )
   &DynamicMaterialParameters(
    Index 0 Param 1=&FloatFromCurve(
     FloatCurve=DataInterface() {propertyValues {Curve {keys [{interpMode=RCIM_Cubic, tangentMode=RCTM_User, time=2.2, value=0}, {interpMode=RCIM_Cubic, tangentMode=RCTM_User, time=3.6, value=1}]}}}
     Scale Curve=&Sine(Normalized Angle=$Emitter.Age, Period=1.437, Scale=2, Bias=0)
    )
    Index 0 Param 2=1
    Index 0 Param 3=-0.1
    Index 1 Param 1=2
    Index 1 Param 2=0.4
    Index 1 Param 3=2
    Index 1 Param 4=0.1
    Module.Param0WriteEnabled=true
    Module.Param0WriteEnabled1=true
    Module.Param1WriteEnabled=false
    Module.Param1WriteEnabled1=false
    Module.Param2WriteEnabled=false
    Module.Param2WriteEnabled1=false
    Module.Param3WriteEnabled=false
    Module.Param3WriteEnabled1=false
    Write Parameter Index 1=true
   )
   &FN_Light_Attributes(
    Color=<rgba 1 0.2 0 1>
    Create Color Attribute=true
    Create Exponent Attribute=true
    Create Radius Attribute=true
    Exponent=2
    Radius=90
    Scale Color=&FloatFromCurve(
     FloatCurve=DataInterface() {propertyValues {Curve {keys [
         {interpMode=RCIM_Cubic, tangentMode=RCTM_User, time=1.1, value=0}
         <at 1.4 0.25>
         {arriveTangent=2.7482, interpMode=RCIM_Cubic, leaveTangent=2.7482, tangentMode=RCTM_User, time=1.9, value=0.25}
         {interpMode=RCIM_Cubic, tangentMode=RCTM_User, time=2.4, value=1}
        ]}}}
     CurveIndex=$Emitter.Age
     Scale Curve=1.29
    )
   )
   &CameraOffset(Camera Offset Amount=&Lerp_Float(A=0, B=-18, Alpha=$System.FN_Integrator001.Value))
  ]
  rendererClasses [&NiagaraLightRendererProperties, &NiagaraMeshRendererProperties]
  renderers [
   {rendererClass=&NiagaraLightRendererProperties, rendererIndex=0}
   {rendererClass=&NiagaraMeshRendererProperties, rendererIndex=1}
  ]
 }
 Center_Ring {
  emitterSpawnScript [&FN_PreserveSkydiveUserVariables() {module=PreserveSkydiveUserVariables}]
  emitterUpdateScript [
   &EmitterState(
    Life Cycle Mode=::NewEnumerator1
    Loop Behavior=::NewEnumerator1
    Loop Duration Mode=::NewEnumerator1
    Reset Age On Awaken=true
    Scalability Mode=::NewEnumerator1
    UseLoopDelay=false
   )
   &SpawnBurst_Instantaneous(Spawn Count=1, Use Spawn Probability=false)
  ]
  particleSpawnScript [
   &FN_SetSpawnGroup(Number of Spawn Groups=1) {module=FN_SetSpawnGroup001}
   &InitializeParticle(
    Color=<rgba 8 1.0649 0.1341 1>
    Mesh Scale=<1.4 -1.4 -2>
    Mesh Scale Mode=::NewEnumerator4
    Sprite Size Mode=::NewEnumerator3
    Uniform Sprite Size=200
    UsePositionOffset=false
    Write Material Random=true
   )
   &InitialMeshOrientation(
    Orientation Axis=<0 0 1>
    Orientation Coordinate Space=::Local
    Orientation Vector=<0 0 1>
    Use Orientation Vector=true
    Use Rotation Vector=false
   )
  ]
  particleUpdateScript [
   &FN_DivingAttributes_v2(Accumulate Dive=0.1, Accumulate Rest=0.1, SkeletalMesh:NiagaraDataInterfaceSkeletalMesh=$User.SkeletalMesh) {module=FN_DivingAttributes_v2001}
   &ParticleState(Kill Particles When Lifetime Has Elapsed=false, Let Infinitely Lived Particles Die When Emitter Deactivates=true)
   &SkeletalMeshLocation(
    Bone / Socket Index=5
    Bone Sampling Mode=::NewEnumerator4
    Inherit Velocity=false
    Interpolate Mesh Sampling=false
    MaskSpawnGroup=false
    Sampled Position Offset=&Lerp_Vector(A=<0 0 32.2>, B=<1 1 25.2>, Alpha=$System.FN_Integrator001.Value)
    Sampled Position Offset Coordinate Space:ENiagaraMeshTransforms=::NewEnumerator2
    Skeletal Mesh=$User.SkeletalMesh
   )
   &SolveForcesAndVelocity(Clamp Velocity=false, Limit Acceleration=false)
   &ScaleMeshSize(Scale Factor=&VectorFromCurve(
     VectorCurve=DataInterface() {propertyValues {
       XCurve {keys [
         {arriveTangent=4.4809, interpMode=RCIM_Cubic, leaveTangent=4.4809, tangentMode=RCTM_User, time=0.1, value=0}
         <at 0.4 0.475>
         {arriveTangent=0.875, interpMode=RCIM_Cubic, leaveTangent=0.875, time=0.6, value=0.475}
         {tangentMode=RCTM_User, time=1, value=1}
        ]}
       YCurve {keys [
         {arriveTangent=4.4809, interpMode=RCIM_Cubic, leaveTangent=4.4809, tangentMode=RCTM_User, time=0.1, value=0}
         <at 0.4 0.475>
         {arriveTangent=0.875, interpMode=RCIM_Cubic, leaveTangent=0.875, time=0.6, value=0.475}
         {tangentMode=RCTM_User, time=1, value=1}
        ]}
       ZCurve {keys [{tangentMode=RCTM_User, time=0, value=1.1}, {tangentMode=RCTM_User, time=1.1, value=1}]}
      }}
     CurveIndex=$Emitter.Age
    ))
   &UpdateMeshOrientation(Rotation Coordinate Space=::NewEnumerator2, Rotation Rate=-1, Rotation Vector=<0 0 1>)
   &ScaleColor(
    Scale Alpha=&FloatFromCurve(
     FloatCurve=DataInterface() {propertyValues {Curve {keys [
         {arriveTangent=4.0516, interpMode=RCIM_Cubic, leaveTangent=4.0516, tangentMode=RCTM_User, time=0, value=0}
         {arriveTangent=0.0961, interpMode=RCIM_Cubic, leaveTangent=0.0961, tangentMode=RCTM_User, time=0.25, value=1}
         {arriveTangent=0.2272, interpMode=RCIM_Cubic, leaveTangent=0.2272, tangentMode=RCTM_User, time=1, value=1}
        ]}}}
     Scale Curve=&Lerp_Float(A=0, B=1, Alpha=$System.FN_Integrator001.Value)
    )
    ScaleA=true
    ScaleRGB=false
   )
   &DynamicMaterialParameters(
    Index 0 Param 2=1
    Index 0 Param 3=&FloatFromCurve(
     FloatCurve=DataInterface() {propertyValues {Curve {keys [{interpMode=RCIM_Cubic, tangentMode=RCTM_User, time=2.2, value=0}, {interpMode=RCIM_Cubic, tangentMode=RCTM_User, time=3.6, value=1}]}}}
     Scale Curve=&Sine(Normalized Angle=$Emitter.Age, Period=1.437, Scale=2, Bias=0)
    )
    Index 1 Param 3=2
    Index 1 Param 4=0.1
    Module.Param0WriteEnabled=true
    Module.Param0WriteEnabled1=true
    Module.Param1WriteEnabled=false
    Module.Param1WriteEnabled1=true
    Module.Param2WriteEnabled=true
    Module.Param2WriteEnabled1=true
    Module.Param3WriteEnabled=false
    Module.Param3WriteEnabled1=false
    Write Parameter Index 1=true
   )
   &FN_Light_Attributes(
    Color=<rgba 1 0.2 0 1>
    Create Color Attribute=true
    Create Exponent Attribute=true
    Create Radius Attribute=true
    Exponent=2
    Radius=90
    Scale Color=&FloatFromCurve(
     FloatCurve=DataInterface() {propertyValues {Curve {keys [
         {interpMode=RCIM_Cubic, tangentMode=RCTM_User, time=1.1, value=0}
         <at 1.4 0.25>
         {arriveTangent=2.7482, interpMode=RCIM_Cubic, leaveTangent=2.7482, tangentMode=RCTM_User, time=1.9, value=0.25}
         {interpMode=RCIM_Cubic, tangentMode=RCTM_User, time=2.4, value=1}
        ]}}}
     CurveIndex=$Emitter.Age
     Scale Curve=1.29
    )
   )
   &CameraOffset(Camera Offset Amount=&Lerp_Float(A=0, B=-18, Alpha=$System.FN_Integrator001.Value))
  ]
  rendererClasses [&NiagaraMeshRendererProperties]
  renderers [
   {rendererClass=&NiagaraMeshRendererProperties, rendererIndex=0}
   {rendererClass=&NiagaraMeshRendererProperties, rendererIndex=1}
  ]
 }
 FN_Contrails_Ribbon_v4 {
  emitterUpdateScript [
   &EmitterScalability(
    Module.bApplySpawnCountScale=false
    Module.bApplySpawnCountScaleByDistanceFraction=false
    Module.bUseMaxDistance=false
    Module.bUseMinDistance=false
   ) {enabled=false}
   &EmitterLifeCycle(CompleteOnInactive=true) {enabled=false}
   &EmitterState()
   &FN_SpawnRibbon_v2(Number of Ribbons=2, Spawn Mode=::NewEnumerator0)
  ]
  particleSpawnScript [
   &FN_InitializeRibbon_v3(Enable Skeletal Position=true, Lifetime=0.7, Ribbon Width=60) {enabled=false}
   &FN_InitializeRibbon_v3(
    Color:LinearColor=/Niagara/DynamicInputs/TypeConversions/MakeLinearColorFromFloat(Value:NiagaraFloat=0.6)
    Enable Skeletal Position=true
    Lifetime=0.7
    Ribbon Width=50
    Sampled Position Coordinate Space:ENiagaraCoordinateSpace=::Local
    Sampled Rotation Coordinate Space:ENiagaraCoordinateSpace=::Local
    Sampled Velocity Coordinate Space:ENiagaraCoordinateSpace=::Local
   ) {module=FN_InitializeRibbon_v3001}
   &AddVelocity(Velocity=&Lerp_Vector(A=<0 0 240>, B=<1 1 400>, Alpha=$System.FN_Integrator001.Value))
   &AddVelocity(Velocity=&MakeCustomVectorFromBool(
     Bool=$Particles.Locked
     True Vector=<0 0 0>
     False Vector=&FN_FrontEndVector3Modifier(Gameplay=<0 0 100>, Front End=<0 0 100>)
    )) {module=AddVelocity001}
   &SampleSkeletalMeshSkeleton(Mesh:NiagaraDataInterfaceSkeletalMesh=$User.SkeletalMesh)
  ]
  particleUpdateScript [
   /Game/Effects/Niagara/Modules/Skydiving/FN_DivingAttributes/FN_DivingAttributes() {enabled=false}
   &SolveForcesAndVelocity(Clamp Velocity=false, Limit Acceleration=false)
   &FN_UpdateRibbon(
    Enable Skeletal Position=true
    Initial Color:LinearColor=$Particles.Initial.Color
    Initial RibbonWidth:NiagaraFloat=$Particles.Initial.RibbonWidth
   )
   &ScaleColor(
    Linear Color Curve=DataInterface() {propertyValues {
      AlphaCurve {keys [<at 0 0.4>,<at 0.7316 0.4>,<at 1 0.1>]}
      BlueCurve {keys [<at 0.0013 0>, <at 0.3545 0 RCIM_Cubic>, {interpMode=RCIM_Cubic, leaveTangent=-0.7216, time=0.9661, value=0}]}
      GreenCurve {keys [
        <at 0.0013 0.1684>
        {arriveTangent=-0.1746, interpMode=RCIM_Cubic, leaveTangent=-0.1746, time=0.3545, value=0.2293}
        {interpMode=RCIM_Cubic, leaveTangent=-0.6443, time=0.9661, value=0}
       ]}
      RedCurve {keys [
        <at 0.0013 0.6962>
        {arriveTangent=-0.7216, interpMode=RCIM_Cubic, leaveTangent=-0.7216, time=0.3545, value=1}
        {arriveTangent=-0.3741, interpMode=RCIM_Cubic, leaveTangent=-0.3741, tangentMode=RCTM_User, time=0.9661, value=0}
       ]}
     }}
    Scale Mode=::NewEnumerator2
   ) {module=ScaleColor001}
   &Material(Base Texture Time:NiagaraFloat=&Multiply_Float(A=$Emitter.Age, B=-1), PostMult=0.5455, Power=0.4408)
   &ScaleColor(
    Scale Alpha=&FloatFromCurve(
     FloatCurve=DataInterface() {propertyValues {Curve {keys [{tangentMode=RCTM_User, time=0.56, value=0}, {tangentMode=RCTM_User, time=1, value=1}]}}}
     CurveIndex=$Emitter.Age
     Scale Curve=&Lerp_Float(A=1, B=1.6, Alpha=$System.FN_Integrator001.Value)
    )
    ScaleA=true
    ScaleRGB=true
   )
  ]
  rendererClasses [&NiagaraRibbonRendererProperties]
  renderers [{rendererClass=&NiagaraRibbonRendererProperties}]
 }
 GPU_Sparks_Center {
  simTarget=GPUComputeSim
  emitterUpdateScript [
   &EmitterState(Life Cycle Mode=::NewEnumerator1, Loop Duration=1.0211, Recalculate Duration Each Loop=true, UseLoopDelay=false)
   &SpawnRate(
    SpawnRate=&Multiply_Float(
     A=&Lerp_Float(A=350, B=500, Alpha=$System.FN_Integrator001.Value)
     B=&FloatFromCurve(
      FloatCurve=DataInterface() {propertyValues {Curve {keys [{interpMode=RCIM_Cubic, tangentMode=RCTM_User, time=0.8, value=0}, {interpMode=RCIM_Cubic, tangentMode=RCTM_User, time=1.7, value=1}]}}}
      CurveIndex=$Emitter.Age
     )
    )
    Use Spawn Probability=false
   )
  ]
  particleSpawnScript [
   &FN_SetSpawnGroup(Number of Spawn Groups=1) {module=FN_SetSpawnGroup001}
   &InitializeParticle(
    Color=<rgba 10 0.98 0.05 1>
    Lifetime Max=0.28
    Lifetime Min=0.16
    Lifetime Mode=::NewEnumerator1
    Sprite Rotation Mode=::NewEnumerator2
    Sprite Size Mode=::NewEnumerator1
    Uniform Sprite Size Max=7
    UsePositionOffset=false
    Write Material Random=true
   )
   &SkeletalMeshLocation(
    Bone / Socket Index=5
    Bone Sampling Mode=::NewEnumerator4
    Inherit Velocity=false
    Interpolate Mesh Sampling=false
    MaskSpawnGroup=false
    Sampled Position Offset=&Lerp_Vector(A=<0 0 32.2>, B=<1 1 25.2>, Alpha=$System.FN_Integrator001.Value)
    Sampled Position Offset Coordinate Space:ENiagaraMeshTransforms=::NewEnumerator3
    Skeletal Mesh=$User.SkeletalMesh
   )
   &ShapeLocation(
    Handle Radius=1
    Large Radius=&FN_FrontEndFloatModifier(Gameplay=48.5, Front End=49.5)
    MaskSpawnGroup=false
    Override Seed=false
    Rotation Mode=::NewEnumerator5
    Scale Mode:ENiagara_ScaleMode=::NewEnumerator1
    Shape Primitive=::NewEnumerator3
    Transform Order:ENiagara_TransformOrder=::NewEnumerator1
   )
   /Niagara/Modules/Update/Velocity/VortexVelocity(
    Module.Use Influence Falloff=false
    Velocity Amount:NiagaraFloat=&Lerp_Float(A=-100, B=-180, Alpha=$User.ForwardDot)
   )
   &AddVelocity(Velocity Speed Scale=&Lerp_Float(A=0.6, B=2, Alpha=$System.Dive))
   &ApplyInitialForces()
   &AddVelocity(
    Module.Use Falloff Distance=false
    Origin Offset=<0 0 40>
    Override Seed=false
    Velocity Mode=::NewEnumerator1
    Velocity Speed=&Lerp_Float(A=20, B=50, Alpha=$System.FN_Integrator001.Value)
   ) {module=AddVelocity001}
  ]
  particleUpdateScript [
   &FN_DivingAttributes_v2(Accumulate Dive=0.1, Accumulate Rest=0.1, SkeletalMesh:NiagaraDataInterfaceSkeletalMesh=$User.SkeletalMesh) {module=FN_DivingAttributes_v2001}
   &ParticleState()
   &ScaleColor(
    Scale Alpha=&FloatFromCurve(FloatCurve=DataInterface() {propertyValues {Curve {keys [<at -0.0046 1.0113 RCIM_Cubic>,<at 1.0011 0.002 RCIM_Cubic>]}}})
    Scale RGB=&VectorFromCurve(VectorCurve=DataInterface() {propertyValues {XCurve {keys [<at 0.5 1>,<at 1 2>]}, YCurve {keys [<at 0.4 1>,<at 1 0.25>]}, ZCurve {keys [<at 0.4 1>,<at 1 0.25>]}}})
    ScaleA=true
    ScaleRGB=false
   )
   &CurlNoiseForce(Noise Frequency=100, Noise Quality / Cost=::NewEnumerator0, Noise Strength=2000, Pan Noise=true, Pan Noise Field=<1.5 -0.2 0.5>)
   &Drag(Drag=&Lerp_Float(A=1, B=0.1, Alpha=$User.ForwardDot), Use Linear Drag=true, Use Rotational Drag=false)
   &InheritSourceMovement(
    Applied Position Scale=&RandomRangeVector(Minimum=<0 0 0>, Maximum=<0 0 -0.1>)
    ApplyForce=false
    ApplyPosition=true
    ApplyPositionFalloff=false
   ) {enabled=false}
   &SolveForcesAndVelocity(Clamp Velocity=false, Limit Acceleration=false)
   &ScaleSpriteSizeBySpeed(Max Scale Factor=<0.4 1.7>, Min Scale Factor=<0.2 0.75>)
   &SolveRotationalForcesAndVelocity()
  ]
  rendererClasses [&NiagaraSpriteRendererProperties]
  renderers [{rendererClass=&NiagaraSpriteRendererProperties}]
 }
 GPU_Sparks_Hands {
  simTarget=GPUComputeSim
  emitterUpdateScript [
   &EmitterState(Delay First Loop Only=true, Life Cycle Mode=::NewEnumerator1, Loop Delay=1.5, Loop Duration=1, UseLoopDelay=true)
   &SpawnRate(SpawnRate=100, Use Spawn Probability=false)
  ]
  particleSpawnScript [
   &InitializeParticle(
    Color=&SelectLinearColorFromArray(
     Array Sampling Mode=::NewEnumerator0
     Direct Array Index=?
     Color Selection Array=DataInterface() {propertyValues {ColorData [<rgba 5 1.0323 0 1>, <rgba 0.401 0.0694 0 1>]}}
     Random Seed=0
    )
    Lifetime Max=0.5
    Lifetime Min=0.4
    Lifetime Mode=::NewEnumerator1
    Sprite Size Mode=::NewEnumerator1
    Uniform Sprite Size Max=3
    Uniform Sprite Size Min=1.5
    UsePositionOffset=false
    Write Material Random=true
   )
   &SkeletalMeshLocation(
    Bone / Socket Index=&RandomRangeInt(Minimum=0, Maximum=1)
    Bone Sampling Mode=::NewEnumerator4
    MaskSpawnGroup=false
    Sampled Position Offset=&FN_FrontEndVector3Modifier(
     Gameplay=&Lerp_Vector(A=<0 12 -13>, B=<0 -8 19>, Alpha=$System.FN_Integrator001.Value)
     Front End=&Lerp_Vector(A=<0 11 -8>, B=<0 -7 19>, Alpha=$System.Dive)
    )
    Sampled Position Offset Coordinate Space:ENiagaraSkelMeshTransforms=::NewEnumerator2
    Skeletal Mesh=$User.SkeletalMesh
   )
   &ShapeLocation(
    MaskSpawnGroup=false
    Offset=<0 0 -9>
    Offset Mode=::NewEnumerator0
    Override Seed=false
    Ring Radius=19
    Shape Primitive=::NewEnumerator4
    UseOwnerScale=true
   )
   &AddVelocity(
    Coordinate Space=::World
    Velocity=&Lerp_Vector(
     A=&RandomRangeVector(Minimum=<0 0 100>, Maximum=<0 0 150>)
     B=&FN_FrontEndVector3Modifier(Gameplay=&RandomRangeVector(Minimum=<0 0 150>, Maximum=<0 0 300>), Front End=<0 0 300>)
     Alpha=$System.FN_Integrator001.Value
    )
   )
   &FN_SetSpawnGroup(Number of Spawn Groups=2)
   &DynamicMaterialParameters(
    Index 0 Param 1=HlslExpression() {hlslExpression="rand(1)"}
    Module.Param0WriteEnabled=true
    Module.Param1WriteEnabled=false
    Module.Param2WriteEnabled=false
    Module.Param3WriteEnabled=false
   )
  ]
  particleUpdateScript [
   &ParticleState()
   &FN_DivingAttributes_v2(
    Accumulate Dive=1
    Accumulate Rest=0
    Bone Index=0
    SkeletalMesh=DataInterface() {propertyValues {
      DefaultMesh=null
      ExcludeBoneName=root
      OptionalComponentClass=null
      PreviewMesh=&M_Med_Soldier_04_Preview
      SkinningMode=SkinOnTheFly
      SoftSourceActor=null
      SourceMode=Default
      UvSetIndex=0
      WholeMeshLOD=-1
      bExcludeBone=false
      bReadDeformedGeometry=true
      bRequireCurrentFrameData=true
      ComponentTags []
      FilteredBones [spine_02]
      FilteredSockets []
      MeshUserParameter {parameter {name=None, typeDefHandle {registeredTypeIndex=1}, varData []}}
      SamplingRegions []
     }}
   )
   &InheritVelocity(Inherited Velocity Amount Scale=<0.5 0.3 -0.05>, Inherited Velocity Speed Limit=0, Speed Limit=true, UseSpeedThreshold=false)
   &CurlNoiseForce(Noise Frequency=10, Noise Quality / Cost=::NewEnumerator0, Noise Strength=1000, Pan Noise=true, Pan Noise Field=<15 10 10>)
   &SolveForcesAndVelocity(Clamp Velocity=false, Limit Acceleration=false)
   &ScaleSpriteSize(Uniform Curve Sprite Scale=DataInterface() {propertyValues {Curve {keys [
        {arriveTangent=6.8904, interpMode=RCIM_Cubic, leaveTangent=6.8904, tangentMode=RCTM_User, time=0, value=0}
        {interpMode=RCIM_Cubic, tangentMode=RCTM_User, time=0.5, value=1}
        {tangentMode=RCTM_User, time=1, value=0}
       ]}}})
   &ScaleSpriteSizeBySpeed(Max Scale Factor=<0.6 2.5>, Min Scale Factor=<0.5 0.85>)
  ]
  rendererClasses [&NiagaraSpriteRendererProperties]
  renderers [{rendererClass=&NiagaraSpriteRendererProperties}]
 }
 Glow_Hands {
  emitterUpdateScript [
   &EmitterState(Life Cycle Mode=::NewEnumerator1, Loop Behavior=::NewEnumerator1, Loop Duration=1, UseLoopDelay=false)
   &SpawnBurst_Instantaneous(Spawn Count=2, Spawn Time=0.03, Use Spawn Probability=false, UseLoopCountLimit=true)
  ]
  particleSpawnScript [
   &FN_SetSpawnGroup(Number of Spawn Groups=2)
   &InitializeParticle(
    Color=<rgba 2 0.4689 0.0476 1>
    Sprite Rotation Mode=::NewEnumerator2
    Sprite Size Mode=::NewEnumerator1
    Uniform Sprite Size Max=180
    Uniform Sprite Size Min=150
    UsePositionOffset=false
    Write Material Random=true
   )
  ]
  particleUpdateScript [
   &SolveForcesAndVelocity(Clamp Velocity=false, Limit Acceleration=false)
   &ParticleState(Kill Particles When Lifetime Has Elapsed=false)
   &SkeletalMeshLocation(
    Bone / Socket Index=$Particles.SpawnGroup
    Bone Sampling Mode=::NewEnumerator4
    Inherit Velocity=false
    Interpolate Mesh Sampling=false
    MaskSpawnGroup=false
    Sampled Position Offset=&Multiply_Vector(
     A=&Lerp_Vector(A=<0 0 1>, B=<0 0 1>, Alpha=$System.Dive)
     B=&Lerp_Vector(A=<0 0 1>, B=<0 0 1>, Alpha=$Particles.NormalizedAge)
    )
    Skeletal Mesh=$User.SkeletalMesh
   )
   &ScaleSpriteSize(Scale Factor=&Vector2DFromFloat(Value=&FloatFromCurve(
      FloatCurve=DataInterface() {propertyValues {Curve {keys [{arriveTangent=1.5832, time=0, value=0.7}, <at 0.1 1>, {leaveTangent=0.5893, time=1, value=0.5}]}}}
      Scale Curve=&Lerp_Float(A=0.9, B=1.8, Alpha=$System.FN_Integrator001.Value)
     )))
   &ScaleColor(
    Scale Alpha=&FloatFromCurve(
     FloatCurve=DataInterface() {propertyValues {Curve {keys [<at 0 0 RCIM_Cubic>,<at 0.15 1>,<at 0.6 1 RCIM_Cubic>]}}}
     CurveIndex=$Emitter.Age
     Scale Curve=0.15
    )
    Scale RGB=<0.5 0.5 0.5>
    ScaleA=true
    ScaleRGB=true
   )
   &CameraOffset(Camera Offset Amount=-15)
   &DynamicMaterialParameters(
    Index 0 Param 2=&FloatFromCurve(FloatCurve=DataInterface() {propertyValues {Curve {keys [
         {interpMode=RCIM_Cubic, tangentMode=RCTM_User, time=0.4, value=0.1}
         {arriveTangent=1.4285, interpMode=RCIM_Cubic, leaveTangent=1.4285, tangentMode=RCTM_User, time=1, value=1}
        ]}}})
    Module.Param0WriteEnabled=true
    Module.Param1WriteEnabled=true
    Module.Param2WriteEnabled=false
    Module.Param3WriteEnabled=false
   )
  ]
  rendererClasses [&NiagaraSpriteRendererProperties]
  renderers [{rendererClass=&NiagaraSpriteRendererProperties}]
 }
 Hands {
  emitterUpdateScript [
   &EmitterState(
    Inactive Response=::NewEnumerator1
    Life Cycle Mode=::NewEnumerator1
    Loop Behavior=::NewEnumerator1
    Loop Duration Mode=::NewEnumerator1
    UseLoopDelay=false
   ) {module=EmitterState001}
   &SpawnBurst_Instantaneous(Use Spawn Probability=false, UseLoopCountLimit=false)
   &SpawnBurst_Instantaneous(Use Spawn Probability=false, UseLoopCountLimit=false) {module=SpawnBurst_Instantaneous001}
  ]
  particleSpawnScript [
   &self:FlipSpliteAnotherHand() {module=FlipSpliteAnotherHand}
   &SpriteFacingAndAlignment(Sprite Facing=<0 0 1>, WriteAlignment=true, WriteFacing=true)
   &FN_SetSpawnGroup(Number of Spawn Groups=2)
   &InitializeParticle(
    Color=<rgba 10 1.4022 0.191 1>
    Mesh Scale=<0.45 0.45 1.25>
    Mesh Scale Mode=::NewEnumerator4
    Sprite Size Mode=::NewEnumerator3
    Uniform Sprite Size=43.5
    UsePositionOffset=false
    Write Material Random=true
   )
   /SetVariables(Particles.UVScale:Vector2f=&Multiply_Vector2D(A=<1 1>, B=HlslExpression() {hlslExpression="float2(Particles.FlipOtherHand, 1.0) /* Custom HLSL! */"})) {isSetParameters=true, module=SetVariables_#5}
   &InitialMeshOrientation(Mesh Orientation Mode=::NewEnumerator1, Use Rotation Vector=false)
  ]
  particleUpdateScript [
   &ParticleState() {enabled=false}
   &SkeletalMeshLocation(
    Bone / Socket Index=$Particles.SpawnGroup
    Bone Sampling Mode=::NewEnumerator4
    Inherit Velocity=false
    "Inherit Velocity (Scale)"=0
    MaskSpawnGroup=false
    Sampled Position Offset=&FN_FrontEndVector3Modifier(
     Gameplay=&Lerp_Vector(A=<0 12 -13>, B=<0 -8 19>, Alpha=$System.FN_Integrator001.Value)
     Front End=&Lerp_Vector(A=<0 11 -10>, B=<0 -7 19>, Alpha=$System.Dive)
    )
    Sampled Position Offset Coordinate Space:ENiagaraMeshTransforms=::NewEnumerator2
    Skeletal Mesh=$User.SkeletalMesh
   )
   /Niagara/Modules/Update/Orientation/V2/OrientMeshToVector(
    Look At Direction:Vector3f=&NormalizeVector(
     Vector To Normalize=&FN_FrontEndVector3Modifier(
      Gameplay=&TransformVector(
       Vector=&Multiply_Vector(
        A=&NormalizeVector(
         Vector To Normalize=&TransformVector(Vector=$Engine.Owner.Velocity, Source Space=::World, Destination Space=::Local)
         Safe Normalize=true
         Safe Fallback Threshold=0
         Safe Fallback Vector=<0 0 0.001>
        )
        B=<1 0 0>
       )
       Destination Space=::World
      )
      Front End=&TransformVector(Vector=&Lerp_Vector(A=<0 0.4 -1>, B=<0 2 -1>, Alpha=$User.ParaGlideLeanAlpha), Destination Space=::Simulation)
     )
     Safe Normalize=true
     Safe Fallback Threshold=0
     Safe Fallback Vector=<0 0 0.001>
    )
    Override Quat=false
    Side Axis Direction:Vector3f=$Engine.Owner.SystemYAxis
    UseOptionalSideAxis=true
    UseOptionalUpAxis=false
   ) {enabled=false}
   &DynamicMaterialParameters(
    Index 0 Param 1=-0.55
    Index 0 Param 2=-1
    Index 0 Param 3=-1
    Index 1 Param 1=1
    Index 1 Param 2=1
    Index 1 Param 3=1
    Index 1 Param 4=&FloatFromCurve(
     FloatCurve=DataInterface() {propertyValues {Curve {keys [{interpMode=RCIM_Cubic, tangentMode=RCTM_User, time=0, value=0}, {interpMode=RCIM_Cubic, tangentMode=RCTM_User, time=0.5, value=1}]}}}
     CurveIndex=$Emitter.Age
     Scale Curve=0.27
    )
    Module.Param0WriteEnabled=true
    Module.Param0WriteEnabled1=true
    Module.Param1WriteEnabled=true
    Module.Param1WriteEnabled1=true
    Module.Param2WriteEnabled=true
    Module.Param2WriteEnabled1=true
    Module.Param3WriteEnabled=true
    Module.Param3WriteEnabled1=true
    Write Parameter Index 1=true
   )
   &CameraOffset(Camera Offset Amount=3.5)
   &ScaleColor(
    Scale Alpha=&Lerp_Float(
     A=1
     B=&Sine(Normalized Angle=$Emitter.Age, Period=0.5, Scale=0.075, Bias=1)
     Alpha=$System.FN_Integrator001.Value
    )
    Scale RGB=&VectorFromFloat(Value=&FloatFromCurve(FloatCurve=DataInterface() {propertyValues {Curve {keys [<at 0 1>,<at 0.5 1>,<at 0.8 4>,<at 0.85 1>]}}}, CurveIndex=$Emitter.Age))
    ScaleA=true
    ScaleRGB=true
   )
   &ScaleMeshSize(Scale Factor=&VectorFromCurve(
     VectorCurve=DataInterface() {propertyValues {
       XCurve {keys [{arriveTangent=1, interpMode=RCIM_Cubic, leaveTangent=1, time=0, value=0}, {tangentMode=RCTM_User, time=0.5, value=1}]}
       YCurve {keys [{arriveTangent=1, interpMode=RCIM_Cubic, leaveTangent=1, time=0, value=0}, {tangentMode=RCTM_User, time=0.5, value=1}]}
       ZCurve {keys [{tangentMode=RCTM_User, time=0, value=1.1}, {tangentMode=RCTM_User, time=0.5, value=1}]}
      }}
     CurveIndex=$Emitter.Age
    ))
   &ScaleSpriteSize(
    Uniform Curve Index=$Emitter.Age
    Uniform Curve Sprite Scale=DataInterface() {propertyValues {Curve {keys [{interpMode=RCIM_Cubic, tangentMode=RCTM_User, time=0, value=0}, {interpMode=RCIM_Cubic, tangentMode=RCTM_User, time=0.5, value=1}]}}}
   )
   &UpdateMeshOrientation(
    Rotation Coordinate Space=::NewEnumerator2
    Rotation Rate=&Multiply_Float(A=0.5, B=$Particles.FlipOtherHand)
    Rotation Vector=<0 0 1>
   )
  ]
  rendererClasses [&NiagaraMeshRendererProperties, &NiagaraSpriteRendererProperties]
  renderers [
   {rendererClass=&NiagaraMeshRendererProperties, rendererIndex=0}
   {rendererClass=&NiagaraMeshRendererProperties, rendererIndex=1}
   {rendererClass=&NiagaraSpriteRendererProperties, rendererIndex=2}
  ]
 }
 Hands001 {
  emitterUpdateScript [
   &EmitterState(
    Inactive Response=::NewEnumerator1
    Life Cycle Mode=::NewEnumerator1
    Loop Behavior=::NewEnumerator1
    Loop Duration Mode=::NewEnumerator1
    UseLoopDelay=false
   ) {module=EmitterState001}
   &SpawnBurst_Instantaneous(Use Spawn Probability=false, UseLoopCountLimit=false)
   &SpawnBurst_Instantaneous(Use Spawn Probability=false, UseLoopCountLimit=false) {module=SpawnBurst_Instantaneous001}
  ]
  particleSpawnScript [
   &self:FlipSpliteAnotherHand() {module=FlipSpliteAnotherHand}
   &FN_SetSpawnGroup(Number of Spawn Groups=2)
   &InitializeParticle(
    Color=<rgba 10 1.4022 0.191 1>
    Mesh Scale=<0.44 0.44 0>
    Mesh Scale Mode=::NewEnumerator4
    Sprite Size Mode=::NewEnumerator3
    Uniform Sprite Size=100
    UsePositionOffset=false
    Write Material Random=true
   )
   /SetVariables(Particles.UVScale:Vector2f=&Multiply_Vector2D(A=<1 1>, B=HlslExpression() {hlslExpression="float2(Particles.FlipOtherHand, 1.0) /* Custom HLSL! */"})) {isSetParameters=true, module=SetVariables_#5}
   &InitialMeshOrientation(Mesh Orientation Mode=::NewEnumerator1, Use Rotation Vector=false)
  ]
  particleUpdateScript [
   &ParticleState() {enabled=false}
   &SkeletalMeshLocation(
    Bone / Socket Index=$Particles.SpawnGroup
    Bone Sampling Mode=::NewEnumerator4
    Inherit Velocity=false
    "Inherit Velocity (Scale)"=0
    MaskSpawnGroup=false
    Sampled Position Offset=&FN_FrontEndVector3Modifier(
     Gameplay=&Lerp_Vector(A=<0 12 -15>, B=<0 -8 19>, Alpha=$System.FN_Integrator001.Value)
     Front End=&Lerp_Vector(A=<0 11 -14>, B=<0 -7 19>, Alpha=$System.Dive)
    )
    Sampled Position Offset Coordinate Space:ENiagaraMeshTransforms=::NewEnumerator2
    Skeletal Mesh=$User.SkeletalMesh
   )
   &OrientMeshToVector(
    Look At Direction:Vector3f=&NormalizeVector(
     Vector To Normalize=&FN_FrontEndVector3Modifier(
      Gameplay=&TransformVector(
       Vector=&Multiply_Vector(
        A=&NormalizeVector(
         Vector To Normalize=&TransformVector(Vector=$Engine.Owner.Velocity, Source Space=::World, Destination Space=::Local)
         Safe Normalize=true
         Safe Fallback Threshold=0
         Safe Fallback Vector=<0 0 0.001>
        )
        B=<1 0 0>
       )
       Destination Space=::World
      )
      Front End=&TransformVector(Vector=&Lerp_Vector(A=<0 0.4 -1>, B=<0 2 -1>, Alpha=$User.ParaGlideLeanAlpha), Destination Space=::Simulation)
     )
     Safe Normalize=true
     Safe Fallback Threshold=0
     Safe Fallback Vector=<0 0 0.001>
    )
    Override Quat=false
    Side Axis Direction:Vector3f=$Engine.Owner.SystemYAxis
    UseOptionalSideAxis=true
    UseOptionalUpAxis=false
   ) {enabled=false}
   &DynamicMaterialParameters(
    Index 1 Param 1=1
    Index 1 Param 2=1
    Index 1 Param 3=1
    Index 1 Param 4=1
    Module.Param0WriteEnabled=true
    Module.Param0WriteEnabled1=true
    Module.Param1WriteEnabled=false
    Module.Param1WriteEnabled1=true
    Module.Param2WriteEnabled=false
    Module.Param2WriteEnabled1=true
    Module.Param3WriteEnabled=false
    Module.Param3WriteEnabled1=true
    Write Parameter Index 1=true
   )
   &CameraOffset(Camera Offset Amount=3.5)
   &ScaleColor(
    Scale Alpha=&Lerp_Float(
     A=1
     B=&Sine(Normalized Angle=$Emitter.Age, Period=0.5, Scale=0.075, Bias=1)
     Alpha=$System.FN_Integrator001.Value
    )
    ScaleA=true
    ScaleRGB=true
   )
   &ScaleMeshSize(Scale Factor=&VectorFromCurve(
     VectorCurve=DataInterface() {propertyValues {
       XCurve {keys [{arriveTangent=1, interpMode=RCIM_Cubic, leaveTangent=1, time=0, value=0}, {tangentMode=RCTM_User, time=0.5, value=1}]}
       YCurve {keys [{arriveTangent=1, interpMode=RCIM_Cubic, leaveTangent=1, time=0, value=0}, {tangentMode=RCTM_User, time=0.5, value=1}]}
       ZCurve {keys [{tangentMode=RCTM_User, time=0, value=1.1}, {tangentMode=RCTM_User, time=0.5, value=1}]}
      }}
     CurveIndex=$Emitter.Age
    ))
   &ScaleMeshSize(Scale Factor=&VectorFromFloat(Value=&Sine(Normalized Angle=$Emitter.Age, Period=1, Scale=0.6, Bias=0.5))) {module=ScaleMeshSize001}
   &UpdateMeshOrientation(
    Rotation Coordinate Space=::NewEnumerator2
    Rotation Rate=&Multiply_Float(A=-0.5, B=$Particles.FlipOtherHand)
    Rotation Vector=<0 0 1>
   )
  ]
  rendererClasses [&NiagaraMeshRendererProperties]
  renderers [{rendererClass=&NiagaraMeshRendererProperties, rendererIndex=0}]
 }
 MeshRibbon_Contrail {
  emitterUpdateScript [
   &EmitterState(
    Inactive Response=::NewEnumerator1
    Life Cycle Mode=::NewEnumerator1
    Loop Behavior=::NewEnumerator1
    Loop Duration Mode=::NewEnumerator1
    UseLoopDelay=false
   ) {module=EmitterState001}
   &SpawnBurst_Instantaneous(Use Spawn Probability=false, UseLoopCountLimit=false)
   &SpawnBurst_Instantaneous(Spawn Time=0.0607, Use Spawn Probability=false, UseLoopCountLimit=false) {module=SpawnBurst_Instantaneous001}
  ]
  particleSpawnScript [
   &FN_SetSpawnGroup(Number of Spawn Groups=2) {module=FN_SetSpawnGroup001}
   &self:FlipSpliteAnotherHand() {module=FlipSpliteAnotherHand}
   &FN_SetSpawnGroup(Number of Spawn Groups=2)
   &InitializeParticle(Mesh Scale=<1 1 0.7>, Mesh Scale Mode=::NewEnumerator4, UsePositionOffset=false, Write Material Random=true)
   /SetVariables(Particles.InitTimeOfs:NiagaraFloat=&Multiply_FloatByInt(Float=0.5, Integer=$Particles.UniqueID)) {isSetParameters=true}
   &InitialMeshOrientation(Mesh Orientation Mode=::NewEnumerator7, Use Rotation Vector=true)
   /SetVariables(Particles.UVScale:Vector2f=&Multiply_Vector2D(A=<1 1>, B=HlslExpression() {hlslExpression="float2(Particles.FlipOtherHand, 1.0) /* Custom HLSL! */"})) {isSetParameters=true, module=SetVariables_#5}
  ]
  particleUpdateScript [
   &ParticleState() {enabled=false}
   &SkeletalMeshLocation(
    Bone / Socket Index=$Particles.SpawnGroup
    Bone Sampling Mode=::NewEnumerator4
    Inherit Velocity=false
    "Inherit Velocity (Scale)"=0
    MaskSpawnGroup=false
    Sampled Position Offset=&FN_FrontEndVector3Modifier(
     Gameplay=&Lerp_Vector(A=<0 15 0>, B=<0 -7 0>, Alpha=$System.Dive)
     Front End=&Lerp_Vector(A=<0 11 0>, B=<0 -7 0>, Alpha=$System.Dive)
    )
    Sampled Position Offset Coordinate Space:ENiagaraMeshTransforms=::NewEnumerator2
    Skeletal Mesh=$User.SkeletalMesh
   )
   &DynamicMaterialParameters(Module.Param0WriteEnabled=true, Module.Param1WriteEnabled=false, Module.Param2WriteEnabled=false, Module.Param3WriteEnabled=false)
   &CameraOffset(Camera Offset Amount=5)
   &ScaleColor(
    Scale Alpha=&Lerp_Float(
     A=0.9
     B=&Sine(Normalized Angle=$Emitter.Age, Period=0.5, Scale=0.075, Bias=1)
     Alpha=$System.FN_Integrator001.Value
    )
    ScaleA=true
    ScaleRGB=true
   )
   &UpdateMeshOrientation(Rotation Rate=&Multiply_Float(A=0.7, B=$Particles.FlipOtherHand), Rotation Vector=<0 0 1>)
   &ScaleMeshSize(Scale Factor=&VectorFromCurve(
     VectorCurve=DataInterface() {propertyValues {
       XCurve {keys [{interpMode=RCIM_Cubic, tangentMode=RCTM_User, time=0, value=0}, {interpMode=RCIM_Cubic, tangentMode=RCTM_User, time=1, value=1}]}
       YCurve {keys [{interpMode=RCIM_Cubic, tangentMode=RCTM_User, time=0, value=0}, {interpMode=RCIM_Cubic, tangentMode=RCTM_User, time=1, value=1}]}
       ZCurve {keys [{interpMode=RCIM_Cubic, tangentMode=RCTM_User, time=0, value=0.4}, {interpMode=RCIM_Cubic, tangentMode=RCTM_User, time=1, value=1}]}
      }}
     CurveIndex=$Emitter.Age
     Scale Curve=&Lerp_Vector(A=<1 1 1>, B=<1.1 1.1 1.5>, Alpha=$System.FN_Integrator001.Value)
    ))
  ]
  rendererClasses [&NiagaraMeshRendererProperties]
  renderers [{rendererClass=&NiagaraMeshRendererProperties, rendererIndex=0}]
 }
 Wave_Center {
  emitterUpdateScript [
   &EmitterState()
   &SpawnRate(SpawnRate=&Lerp_Float(A=0, B=60, Alpha=$System.FN_Integrator001.Value), Use Spawn Probability=false)
  ]
  particleSpawnScript [
   &InitializeParticle(
    Color=<rgba 1.1 0.3783 0 0.15>
    Lifetime=&RandomRangeFloat(Minimum=0.3, Maximum=0.4)
    Sprite Size Max=<10 50>
    Sprite Size Min=<10 30>
    Sprite Size Mode=::NewEnumerator2
    UsePositionOffset=false
    Write Material Random=true
   )
   /SetVariables(Particles.Position=&Add_Vector(A=Linked() {linkedVariable {name=Engine.Owner.Position, type=NiagaraPosition}}, B=<0 0 55>)) {isSetParameters=true}
   &TorusLocation(Align Torus With Axis=false, Handle Radius=0, Large Radius=38, MaskSpawnGroup=false, Module.Override Randomness=false)
   /SetVariables(Particles.RandomFrame=&RandomRangeInt(Minimum=0, Maximum=63)) {isSetParameters=true}
  ]
  particleUpdateScript [
   &ParticleState()
   &InheritSourceMovement(
    Applied Force Scale=<-0.5 -0.5 0>
    ApplyForce=true
    ApplyForceFalloff=false
    ApplyPosition=true
    ApplyPositionFalloff=false
    Module.LimitForceSpeed=false
   ) {module=InheritSourceMovement001}
   &AddVelocityFromPoint(
    Module.Use Falloff Distance=false
    Origin Offset=<0 0 22>
    Velocity Strength=&FloatFromCurve(
     FloatCurve=DataInterface() {propertyValues {Curve {keys [
         {arriveTangent=-1.0543, interpMode=RCIM_Cubic, time=0.25, value=1}
         {arriveTangent=-5, interpMode=RCIM_Cubic, time=0.6, value=0}
         <at 0.7 -1 RCIM_Cubic>
        ]}}}
     Scale Curve=13
    )
   ) {enabled=false}
   &PointForce(
    Force Falloff Distance=155
    Force Origin=&Add_Vector(A=/Niagara/DynamicInputs/Helpers/SimulationPosition(), B=<0 0 50>)
    Force Strength=1500
    Module.Use Falloff=true
    Module.UseLinearForce=false
   )
   &GravityForce(Gravity=&MakeVector(
     X=0
     Y=0
     Z=&FloatFromCurve(
      FloatCurve=DataInterface() {propertyValues {Curve {keys [<at 0.4 0 RCIM_Cubic>, {arriveTangent=0.5303, interpMode=RCIM_Cubic, leaveTangent=0.5303, tangentMode=RCTM_User, time=0.5, value=1}]}}}
      Scale Curve=2500
     )
    ))
   &SolveForcesAndVelocity(Clamp Velocity=false, Limit Acceleration=false)
   &ScaleColor(
    Scale Alpha=&FloatFromCurve(
     FloatCurve=DataInterface() {propertyValues {Curve {keys [
         {arriveTangent=-0.7151, interpMode=RCIM_Cubic, leaveTangent=-0.7151, tangentMode=RCTM_User, time=0.2, value=0}
         {arriveTangent=-2.5, interpMode=RCIM_Cubic, leaveTangent=-2.5, time=0.6, value=1}
         {arriveTangent=-1.6379, interpMode=RCIM_Cubic, leaveTangent=-1.6379, tangentMode=RCTM_User, time=1, value=0}
        ]}}}
     Scale Curve=&Lerp_Float(A=0.2, B=0.5, Alpha=$System.FN_Integrator001.Value)
    )
    Scale RGB=&MakeVectorFromLinearColorRGB(Color=&ColorFromCurve(ColorCurve=DataInterface() {propertyValues {
        AlphaCurve {keys [<at 0.3745 0>, {arriveTangent=-5.0984, interpMode=RCIM_Cubic, leaveTangent=-5.0984, time=0.529, value=1}, <at 0.7251 0>]}
        BlueCurve {keys [<at 0.0463 0>]}
        GreenCurve {keys [<at 0.0463 3.2194>]}
        RedCurve {keys [<at 0.0463 5>]}
       }}))
    ScaleA=true
    ScaleRGB=true
   )
   &ScaleSpriteSize(Scale Factor=&Vector2DFromFloat(Value=&FloatFromCurve(
      FloatCurve=DataInterface() {propertyValues {Curve {keys [{arriveTangent=-0.2263, interpMode=RCIM_Cubic, leaveTangent=-0.2263, tangentMode=RCTM_User, time=0, value=0}, <at 1 1 RCIM_Cubic>]}}}
      Scale Curve=3.5
     )))
   &FN_SubUVAnimation(End Frame=&Add_Integer(A=$Particles.RandomFrame, B=24), Start Frame=$Particles.RandomFrame)
  ]
  rendererClasses [&NiagaraSpriteRendererProperties]
  renderers [{rendererClass=&NiagaraSpriteRendererProperties}]
 }
}
summary {
 systemName=NS_Contrail_Dive_Mystic
 emitters [
  {emitterName=Center, rendererClasses [&NiagaraLightRendererProperties, &NiagaraMeshRendererProperties]}
  {emitterName=Hands, rendererClasses [&NiagaraMeshRendererProperties, &NiagaraSpriteRendererProperties]}
  {emitterName=GPU_Sparks_Hands, simTarget=GPUComputeSim, rendererClasses [&NiagaraSpriteRendererProperties]}
  {emitterName=FN_Contrails_Ribbon_v4, rendererClasses [&NiagaraRibbonRendererProperties]}
  {emitterName=Wave_Center, rendererClasses [&NiagaraSpriteRendererProperties]}
  {emitterName=Glow_Hands, rendererClasses [&NiagaraSpriteRendererProperties]}
  {emitterName=MeshRibbon_Contrail, rendererClasses [&NiagaraMeshRendererProperties]}
  {emitterName=GPU_Sparks_Center, simTarget=GPUComputeSim, rendererClasses [&NiagaraSpriteRendererProperties]}
  {emitterName=Hands001, rendererClasses [&NiagaraMeshRendererProperties]}
  {emitterName=Center_Ring, rendererClasses [&NiagaraMeshRendererProperties]}
 ]
 userVariables [
  User.bIsFrontEnd {type=NiagaraBool}
  User.bIsFrontEndPreview {type=NiagaraBool}
  User.ColorChange {type=LinearColor, description="", defaultValue {struct=LinearColor, value=<rgba 1 1 1 1>}}
  User.ForwardDot {type=NiagaraFloat, defaultValue=0, description=""}
  User.ParaGlideLeanAlpha {type=NiagaraFloat, defaultValue=0, description=""}
  User.PlayerSpeed {type=NiagaraFloat, defaultValue=0, description=""}
  User.RibbonIDOffset {type=NiagaraInt32}
  User.SkeletalMesh {
   type=NiagaraDataInterfaceSkeletalMesh
   defaultValue {
    struct=NiagaraExt_VariableValue_DataInterface
    value {dataInterface=&self:NiagaraDataInterfaceSkeletalMesh_0}
   }
  }
 ]
}
